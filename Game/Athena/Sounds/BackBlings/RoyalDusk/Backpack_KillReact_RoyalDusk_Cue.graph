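# Sound cue for the kill reaction of the RoyalDusk back bling.
# A random node with equal weights picks one of three kill-react wave variations each time it plays.
# It routes through the back bling sound class at reduced volume, with close-range pet attenuation.
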SoundCue /Game/Athena/Sounds/BackBlings/RoyalDusk/Backpack_KillReact_RoyalDusk_Cue
root=Random_0
nodes [
 {id=Random_0, children [WavePlayer_0,WavePlayer_1,WavePlayer_2], properties {weights [1,1,1]}}
 {id=WavePlayer_0, properties {soundWaveAssetPtr=&/Backpack_KillReact_RoyalDusk_03}}
 {id=WavePlayer_1, properties {soundWaveAssetPtr=&/Backpack_KillReact_RoyalDusk_01}}
 {id=WavePlayer_2, properties {soundWaveAssetPtr=&/Backpack_KillReact_RoyalDusk_02}}
]
properties {
 attenuationSettings=&Fort_Pets_Close_Athena_Att
 maxDistance=1800.0001
 soundClassObject=/Game/Sounds/FortSoundClassesAndModes/Fort_Backblings
 volumeMultiplier=0.42
}
